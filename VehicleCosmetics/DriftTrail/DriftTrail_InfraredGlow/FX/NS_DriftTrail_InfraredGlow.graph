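# Niagara drift trail cosmetic "InfraredGlow" for vehicles, drawing two ribbon trails and mesh wraps on the rear wheels.
# The ribbon emitter has three renderers, fed separate background, color and stripe attributes whose over-life curves shift warm to cool like a thermal gradient while width tapers.
# The wrap emitter bursts two meshes, re-snaps them to skeletal sockets and scales them per wheel.
# Both emitters use per-wheel tire FX masking; painted-variant overrides are disabled.
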NiagaraSystem /VehicleCosmetics/DriftTrail/DriftTrail_InfraredGlow/FX/NS_DriftTrail_InfraredGlow
dependencies {
 usedDataInterfaces [
  &NiagaraDataInterfaceVectorCurve
  &NiagaraDataInterfaceCurve
  &NiagaraDataInterfaceColorCurve
  &NiagaraDataInterfaceSkeletalMesh
 ]
 usedDynamicInputs []
 usedModules [
  &self:SystemSpawnScript.SystemScriptSource.SystemScriptGraph.NiagaraNodeAssignment_0.SetVariables_#1
  &FN_UnpackVehicleFlags
  /VehicleCosmetics/FX/Niagara/Modules/VC_SetVehicleSpeedAttribs
  &SystemState
 ]
}
emitters {
 NE_VC_DriftTrail_Ribbon_Parent001 {
  emitterUpdateScript [
   &EmitterState(Inactive Response=::NewEnumerator1, Life Cycle Mode=::NewEnumerator1, Loop Duration=1, UseLoopDelay=false)
   &VC_RateQualityHandler(Epic Spawn Rate=30, High Spawn Rate=60, Low Spawn Rate=30, Use Custom Rate?=true) {enabled=false}
   &VC_SpawnRibbon_v1(Number of Ribbons=2, Spawn Time=-1, SpawnRate=60)
  ]
  particleSpawnScript [
   &VC_PaintedVariantsOverride(
    Black=<rgba 0.0191 0.0191 0.0191 0.5>
    BurntSienna=<rgba 0.0642 0.0198 0.0079 0.7>
    Cobalt=<rgba 0 0.0185 1 0.7>
    Crimson=<rgba 0.4635 0 0.0126 0.6>
    DefaultAlpha=0.5
    DefaultColor=<rgba 0 0 1 0.5>
    ForestGreen=<rgba 0 0.5 0.0253 0.5>
    Grey=<rgba 0.2031 0.2031 0.2031 0.5>
    Lime=<rgba 0.2943 0.5885 0 0.5>
    Orange=<rgba 0.6724 0.2346 0 0.6>
    Pink=<rgba 0.5052 0.0605 0.3435 0.5>
    Purple=<rgba 0.2789 0 0.6795 0.5>
    Saffron=<rgba 0.9 0.69 0 0.5>
    SkyBlue=<rgba 0 0.357 0.5365 0.5>
    TitaniumWhite=<rgba 0.75 0.75 0.75 0.5>
   ) {enabled=false}
   &VC_InitializeRibbon_v1(
    Color=<rgba 0 0 1 1>
    Custom Offset=true
    Enable Skeletal Position=true
    Lifetime=1.2
    Local Vector Offset=&AddVectorToPosition(Position=$User.offset, Vector=<-10 0 -25>)
    Offset Amount=1
    Ribbon Width=80
   )
   &AddVelocity(Velocity=<-1000 0 0>) {module=AddVelocity001}
  ]
  particleUpdateScript [
   &FN_DisableTireFX(Mask Index Attribute:NiagaraInt32=$Particles.SpawnGroup, Mask Index FL=2, Mask Index FR=3, Mask Index RL=0, Mask Index RR=1)
   &ParticleState()
   &InheritSourceMovement(
    Applied Position Scale=&Multiply_VectorByFloat(Vector=<1 1 1>, Float=1)
    ApplyForce=false
    ApplyPosition=true
    ApplyPositionFalloff=false
   )
   &SolveForcesAndVelocity(Clamp Velocity=false, Limit Acceleration=false)
   &VC_UpdateRibbon_v1(Enable Skeletal Position=true)
   /SetVariables(
    Particles.BackgroundColor:LinearColor=&MultiplyLinearColor(
     Linear Color=$Particles.Color
     Scale Factor=&MakeVector4FromVectorAndFloat(
      "Vector (RGB)"=&VectorFromCurve(
       VectorCurve=DataInterface() {propertyValues {XCurve {keys [<at 0 1>]}, YCurve {defaultValue=-0.0176, keys [<at 0.574 0.9967>]}, ZCurve {keys [<at 0.0611 0.9936>,<at 0.582 1.0005>]}}}
       Scale Curve=<0.5 0.5 0.5>
      )
      "Float (Alpha)"=&FloatFromCurve(
       FloatCurve=DataInterface() {propertyValues {Curve {
          defaultValue=1
          keys [<at 0.0045 0.0069>, <at 0.18 1>, {tangentMode=RCTM_User, time=0.5796, value=1}, {tangentMode=RCTM_User, time=1.012, value=-0.0421}]
         }}}
       Scale Curve=1.5
      )
     )
    )
    Particles.ColorsColor:LinearColor=&MultiplyLinearColor(
     Linear Color=&ColorFromCurve(ColorCurve=DataInterface() {propertyValues {
        AlphaCurve {keys [<at 0 0>,<at 0.0398 1>,<at 1 0 RCIM_Cubic>]}
        BlueCurve {keys [<at 0.1823 0 RCIM_Cubic>,<at 0.356 0>,<at 0.5116 0.0012>,<at 0.7747 1>,<at 0.9657 1 RCIM_Cubic>]}
        GreenCurve {keys [<at 0.1823 0 RCIM_Cubic>,<at 0.356 0.5333>,<at 0.5116 0.5>,<at 0.7747 0.6667>,<at 0.9657 0.2014 RCIM_Cubic>]}
        RedCurve {keys [<at 0.1823 0.5 RCIM_Cubic>,<at 0.356 0.8>,<at 0.5116 0.0951>,<at 0.7747 0>,<at 0.9657 0 RCIM_Cubic>]}
       }})
     Scale Factor=&MakeVector4FromVectorAndFloat("Vector (RGB)"=<1 1 1>, "Float (Alpha)"=1)
    )
    Particles.ColorsWidth:NiagaraFloat=&FloatFromCurve(
     FloatCurve=DataInterface() {propertyValues {Curve {keys [<at 0.4 1 RCIM_Cubic>, {arriveTangent=-1.8454, interpMode=RCIM_Cubic, leaveTangent=-1.8454, tangentMode=RCTM_User, time=0.98, value=0}]}}}
     Scale Curve=25
    )
    Particles.StripesColor:LinearColor=&MultiplyLinearColor(
     Linear Color=&ColorFromCurve(ColorCurve=DataInterface() {propertyValues {
        AlphaCurve {keys [<at 0.0059 0>,<at 0.0435 1>,<at 1 0 RCIM_Cubic>]}
        BlueCurve {keys [
          <at 0.1305 0>
          {arriveTangent=0.6652, interpMode=RCIM_Cubic, time=0.2082, value=0}
          <at 0.356 0.2>
          <at 0.5045 0.04>
          <at 0.838 1>
          <at 0.9657 1 RCIM_Cubic>
         ]}
        GreenCurve {keys [
          <at 0.1305 0>
          {arriveTangent=2.705, interpMode=RCIM_Cubic, time=0.2082, value=0.15}
          <at 0.356 0.8133>
          <at 0.5045 0.4>
          <at 0.838 0.5471>
          <at 0.9657 0.3611 RCIM_Cubic>
         ]}
        RedCurve {keys [
          <at 0.1305 0.5029>
          {arriveTangent=1.6533, interpMode=RCIM_Cubic, time=0.2082, value=0.9}
          <at 0.356 1>
          <at 0.5045 0.1078>
          <at 0.838 0.2>
          <at 0.9657 0.2 RCIM_Cubic>
         ]}
       }})
     Scale Factor=&MakeVector4FromVectorAndFloat("Vector (RGB)"=<1 1 1>, "Float (Alpha)"=1)
    )
   ) {isSetParameters=true}
   &ScaleRibbonWidth(Ribbon Width Scale=&FloatFromCurve(FloatCurve=DataInterface() {propertyValues {Curve {keys [<at 0.6 1>,<at 0.9 0.3>]}}}))
  ]
  rendererClasses [&NiagaraRibbonRendererProperties]
  renderers [
   {rendererClass=&NiagaraRibbonRendererProperties}
   {rendererClass=&NiagaraRibbonRendererProperties, rendererIndex=1}
   {rendererClass=&NiagaraRibbonRendererProperties, rendererIndex=2}
  ]
 }
 NE_VC_DriftTrail_WheelWrap_Parent {
  emitterSpawnScript [
   /SetVariables(Emitter.No Painted:NiagaraBool=&Boolean_NotOperation(Bool=$User.bIsPainted)) {isSetParameters=true}
   /SetVariables(Emitter.Painted:NiagaraBool=&Boolean_NotOperation(Bool=$StackContext.No Painted)) {isSetParameters=true}
  ]
  emitterUpdateScript [
   &EmitterState(
    Inactive Response=::NewEnumerator1
    Life Cycle Mode=::NewEnumerator1
    Loop Behavior=::NewEnumerator1
    Loop Duration Mode=::NewEnumerator1
    UseLoopDelay=false
   )
   &SpawnBurst_Instantaneous(Spawn Count=2, Use Spawn Probability=false, UseLoopCountLimit=false)
  ]
  particleSpawnScript [
   &VC_PaintedVariantsOverride(
    Black=<rgba 0 0 0 3>
    BurntSienna=<rgba 0.1 0.005 0 5>
    Cobalt=<rgba 0 0.0185 1 1>
    Crimson=<rgba 0.4635 0 0.0126 1>
    DefaultAlpha=0.5
    ForestGreen=<rgba 0 0.8 0.0404 1>
    Grey=<rgba 0.2031 0.2031 0.2031 3>
    Lime=<rgba 0.5 1 0 1>
    Orange=<rgba 0.4 0.1395 0 2>
    Pink=<rgba 1 0.3 0.7455 1.2>
    R1 Black:LinearColor=<rgba 0.1 0.1 0.1 1>
    R1 BurntSienna:LinearColor=<rgba 0.5 0.025 0 1>
    R1 Cobalt:LinearColor=<rgba 0.1035 0.1145 0.7 1>
    R1 Crimson:LinearColor=<rgba 0.4635 0.0371 0.0486 1>
    R1 ForestGreen:LinearColor=<rgba 0 1 0.0505 1>
    R1 Grey:LinearColor=<rgba 0.8 0.8 0.8 1>
    R1 Lime:LinearColor=<rgba 0.75 1.5 0 1>
    R1 Orange:LinearColor=<rgba 1.5 0.5232 0 1>
    R1 Pink:LinearColor=<rgba 1 0.3 0.7455 1.2>
    R1 Purple:LinearColor=<rgba 0.106 0.008 0.4 2>
    R1 Saffron:LinearColor=<rgba 1 0.7667 0 1>
    R1 SkyBlue:LinearColor=<rgba 0.3 0.9431 1 1>
    R1 TitaniumWhite:LinearColor=<rgba 1 1 1 3>
    SkyBlue=<rgba 0 0.5323 0.8 1>
    TitaniumWhite=<rgba 1 1 1 2>
    Use R1 Painted Overrides=true
   ) {enabled=false}
   &InitializeParticle(
    Lifetime=5
    Mesh Scale Mode=::NewEnumerator4
    Sprite Rotation Mode=::NewEnumerator2
    Sprite Size Mode=::NewEnumerator3
    UsePositionOffset=false
    Write Material Random=true
   )
   &PartitionParticles()
   &SkeletalMeshLocation(
    Bone / Socket Index=&Add_Integer(A=$Particles.PartitionParticles.Partition, B=4)
    Bone / Socket Sampling Mode=::NewEnumerator5
    Interpolate Mesh Sampling=false
    MaskSpawnGroup=false
    Mesh Sampling Type=::NewEnumerator6
    Skeletal Mesh=$System.SkeletalMesh
   )
   &InitialMeshOrientation(Mesh Orientation Mode=::NewEnumerator7, Rotation=<0 0 0.5>, Use Rotation Vector=true) {enabled=false}
   &Color(Color=<rgba 0 1 0 1>) {enabled=false}
  ]
  particleUpdateScript [
   &FN_DisableTireFX(
    Mask Index Attribute:NiagaraInt32=$Particles.PartitionParticles.Partition
    Mask Index FL=2
    Mask Index FR=3
    Mask Index RL=0
    Mask Index RR=1
   )
   &VC_WheelWrapScale(Wheel Index Attribute:NiagaraInt32=&Add_Integer(A=$Particles.PartitionParticles.Partition, B=2))
   &ParticleState(Kill Particles When Lifetime Has Elapsed=false)
   &UpdateMeshOrientation(Rotation Coordinate Space=::NewEnumerator2, Rotation Vector=<0 1 0>) {enabled=false}
   &SkeletalMeshLocation(
    Bone / Socket Index=&Add_Integer(A=$Particles.PartitionParticles.Partition, B=4)
    Bone / Socket Sampling Mode=::NewEnumerator5
    MaskSpawnGroup=false
    Mesh Sampling Type=::NewEnumerator6
    Skeletal Mesh=$System.SkeletalMesh
   ) {module=SkeletalMeshLocation001}
   &SolveForcesAndVelocity(Clamp Velocity=false, Limit Acceleration=false)
   &ScaleColor(
    Scale RGB=&VectorFromFloat(Value=&Waveform("Waveform [1]"=::NewEnumerator0, "[1] Phase Scale"=5, "[1] Frequency"=0.5, "[1] Amplitude Min/Max"=<50 100>))
    ScaleA=true
    ScaleRGB=true
   ) {enabled=false}
  ]
  rendererClasses [&NiagaraMeshRendererProperties]
  renderers [
   {rendererClass=&NiagaraMeshRendererProperties, rendererIndex=0}
   {rendererClass=&NiagaraMeshRendererProperties, rendererIndex=1}
  ]
 }
}
summary {
 systemName=NS_DriftTrail_InfraredGlow
 emitters [
  {emitterName=NE_VC_DriftTrail_WheelWrap_Parent, rendererClasses [&NiagaraMeshRendererProperties]}
  {emitterName=NE_VC_DriftTrail_Ribbon_Parent001, rendererClasses [&NiagaraRibbonRendererProperties]}
 ]
 userVariables [
  User.bIsFrontendPreview {type=NiagaraBool}
  User.bIsPainted {type=NiagaraBool}
  User.ColorVariantID {type=NiagaraInt32}
  User.DTCarOffset {type=NiagaraPosition, description="", defaultValue {}}
  User.offset {type=NiagaraPosition, description="", defaultValue {}}
  User.PaintedColor {type=LinearColor, description="", defaultValue {struct=LinearColor, value=<rgba 1 1 1 1>}}
  User.PaintedScalar {type=NiagaraFloat, defaultValue=1, description=""}
  User.VehicleFlags {type=NiagaraInt32, defaultValue=806289408}
  User.Wheel_Radius_Array {
   type=NiagaraDataInterfaceArrayFloat
   defaultValue {
    struct=NiagaraExt_VariableValue_DataInterface
    value {
     dataInterface=&self:NiagaraDataInterfaceArrayFloat_1
     dataInterfaceClass=&NiagaraDataInterfaceArrayFloat
    }
   }
  }
  User.Wheel_Width_Array {
   type=NiagaraDataInterfaceArrayFloat
   defaultValue {
    struct=NiagaraExt_VariableValue_DataInterface
    value {
     dataInterface=&self:NiagaraDataInterfaceArrayFloat_0
     dataInterfaceClass=&NiagaraDataInterfaceArrayFloat
    }
   }
  }
  User.WheelTouchingArray {
   type=NiagaraDataInterfaceArrayFloat
   defaultValue {
    struct=NiagaraExt_VariableValue_DataInterface
    value {
     dataInterface=&self:NiagaraDataInterfaceArrayFloat_2
     dataInterfaceClass=&NiagaraDataInterfaceArrayFloat
    }
   }
  }
 ]
}
